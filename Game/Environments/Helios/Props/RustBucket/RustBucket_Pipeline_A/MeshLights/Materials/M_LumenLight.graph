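# M_LumenLight is an unlit emissive material for the RustBucket pipeline mesh lights, flagged for dynamic area lighting.
# It multiplies a Light Color by the MPC_PipelineLight Brightness Multiplier and a pulse sampled from the C_LightPulseOn atlas curve.
# Curve time is the time since TimeFromBP, divided by Pulse Speed and clamped.
# TimeFromBP and LightOn use custom primitive data, with rounded LightOn gating the result on or off.
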Material /Game/Environments/Helios/Props/RustBucket/RustBucket_Pipeline_A/MeshLights/Materials/M_LumenLight
graph {
 properties {
  bAutomaticallySetUsageInEditor=false
  bUseEmissiveForDynamicAreaLighting=true
  pixelDepthOffsetMode=PDOM_Legacy
  shadingModel=MSM_Unlit
  usedShadingModels=Unlit
 }
 nodes [
  n1 = VectorParameter() {parameterName=Light Color, defaultValue=<rgba 100 85.37 0 1>}
  n2 = CollectionParameter() {
   collection=&/MPC_PipelineLight
   parameterName=Brightness Multiplier
  }
  n3 = Subtract(A=Time(), B=ScalarParameter() {bUseCustomPrimitiveData=true, parameterName=TimeFromBP, primitiveDataIndex=1})
  n4 = ScalarParameter() {defaultValue=0.5, parameterName=Pulse Speed}
  n5 = CurveAtlasRowParameter(CurveTime=Saturate(None=Divide(A=@n3, B=@n4))) {
   atlas=/Game/Environments/Helios/Props/RustBucket/RustBucket_Pipeline_A/MeshLights/Curves/CA_PipelineLights
   curve=/Game/Environments/Helios/Props/RustBucket/RustBucket_Pipeline_A/MeshLights/Curves/C_LightPulseOn
  }
  n6 = ScalarParameter() {bUseCustomPrimitiveData=true, parameterName=LightOn, primitiveDataIndex=2}
  Multiply(A=Multiply(A=Multiply(A=@n1.RGB, B=@n2), B=@n5.R), B=Round(None=Saturate(None=@n6)))
 ]
}
