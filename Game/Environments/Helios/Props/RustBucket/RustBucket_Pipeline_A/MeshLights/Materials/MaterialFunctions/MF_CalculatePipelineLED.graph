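# Material function that draws one sample of a procedural LED light grid for the Helios RustBucket pipeline mesh lights, outputting Spread and Core masks.
# Radial falloff around each cell center forms the dots, with Sample Offset selecting the evaluated cell.
# Row distance from center shifts columns into repeating lit and unlit chevron segments, handling odd or even row counts and masking out-of-range rows.
# Core adds fake depth by parallax-shifting UVs along the tangent-space camera vector.
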MaterialFunction /Game/Environments/Helios/Props/RustBucket/RustBucket_Pipeline_A/MeshLights/Materials/MaterialFunctions/MF_CalculatePipelineLED
graph {nodes [
  n1 = FunctionInput() {inputName=Light Grid UVs, inputType=FunctionInput_Vector2}
  n2 = FunctionInput() {bUsePreviewValueAsDefault=true, inputName=Sample Offset, inputType=FunctionInput_Vector2}
  n3 = Floor(None=Subtract(A=@n1, B=@n2))
  n4 = Length(None=Subtract(A=Subtract(A=@n1, B=@n3)) {constB=0.5}) {is=class MaterialExpressionLength}
  n5 = FunctionInput() {inputName=Light Spread, inputType=FunctionInput_Scalar, previewValue=<1 0 0 1>}
  n6 = OneMinus(None=Saturate(None=Divide(A=@n4, B=@n5)))
  n7 = FunctionInput() {inputName=Light Spread Power, inputType=FunctionInput_Scalar}
  n8 = MaterialFunctionCall(Float2=@n3) {materialFunction=&BreakOutFloat2Components}
  n9 = FunctionInput() {inputName=Light Count Y, inputType=FunctionInput_Scalar, previewValue=<5 0 0 1>}
  n10 = OneMinus(None=Modulo!(A=@n9, B=Constant() {r=2}))
  n11 = Subtract(A=@n9, B=@n10)
  n12 = Subtract(A=Modulo(A=@n8.G, B=@n9), B=Floor(None=Divide(A=@n11)))
  n13 = Saturate(None=@n12)
  n14 = Add(A=@n12, B=Multiply(A=Subtract(A=@n13), B=@n10))
  n15 = Modulo(
   A=Subtract(A=@n8.R, B=Abs(None=@n14))
   B=FunctionInput() {inputName=Segment Width Count, inputType=FunctionInput_Scalar, previewValue=<4 0 0 1>}
  )
  n16 = Subtract(A=FunctionInput() {inputName=Unlit Width Count, inputType=FunctionInput_Scalar, previewValue=<2 0 0 1>})
  n17 = Saturate(None=Subtract(A=@n8.G, B=Subtract(A=@n9)))
  n18 = Saturate(None=Multiply(A=@n8.G) {constB=-1})
  n19 = Multiply(A=Saturate(None=Subtract(A=@n15, B=@n16)), B=OneMinus(None=Max(A=@n17, B=@n18)))
  FunctionOutput(None=Multiply(A=Power(Base=@n6, Exp=@n7), B=@n19)) {outputName=Spread}
  n20 = CameraVectorWS()
  n21 = ComponentMask(None=Normalize(VectorInput=Transform(None=@n20) {transformSourceType=TRANSFORMSOURCE_World, transformType=TRANSFORM_Tangent})) {g=true, r=true}
  n22 = Add(A=@n1, B=Multiply(A=Multiply(A=@n21) {constB=-1}, B=FunctionInput() {inputName=Core Depth, inputType=FunctionInput_Scalar}) {constB=0.5})
  n23 = Floor(None=Subtract(A=@n22, B=@n2))
  n24 = Length(None=Subtract(A=Subtract(A=@n22, B=@n23)) {constB=0.5})
  n25 = FunctionInput() {inputName=Light Core Width, inputType=FunctionInput_Scalar}
  n26 = OneMinus(None=Saturate(None=Divide(A=@n24, B=@n25)))
  n27 = FunctionInput() {inputName=Light Core Power, inputType=FunctionInput_Scalar}
  FunctionOutput(None=Multiply(A=Power(Base=@n26, Exp=@n27), B=@n19)) {outputName=Core}
 ]}
