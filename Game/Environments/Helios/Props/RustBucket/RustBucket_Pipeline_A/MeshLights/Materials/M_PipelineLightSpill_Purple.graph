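# Translucent unlit material that draws the soft purple light spill cast by the animated LEDs on the RustBucket pipeline prop.
# A time-stepped UV grid, flippable by per-primitive Direction data, feeds three offset MF_CalculatePipelineLED calls whose summed spread forms the glow.
# Edge, bracket and view-angle masks shape the spill.
# A collection brightness multiplier and a curve-atlas pulse, timed from per-primitive TimeFromBP, scale it as the lights switch on.
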Material /Game/Environments/Helios/Props/RustBucket/RustBucket_Pipeline_A/MeshLights/Materials/M_PipelineLightSpill_Purple
graph {
 properties {
  bAutomaticallySetUsageInEditor=false
  blendMode=BLEND_Translucent
  pixelDepthOffsetMode=PDOM_Legacy
  shadingModel=MSM_Unlit
  usedShadingModels=Unlit
 }
 nodes [
  n1 = NamedRerouteUsage()
  n2 = AppendVector(A=ConstantBiasScale(None=ComponentMask(None=@n1) {g=true, r=true}) {bias=-0.5, scale=2}, B=Constant())
  ConstantBiasScale(None=ComponentMask(None=Transform(None=@n2) {transformType=TRANSFORM_Camera}) {b=true})
  n3 = NamedRerouteUsage()
  n4 = ComponentMask(None=@n3) {r=true}
  n5 = Constant()
  n6 = Constant() {r=1}
  n7 = Multiply(A=ScalarParameter() {defaultValue=0.65, parameterName=Bloom Edge Padding}) {constB=-1}
  n8 = Subtract(A=ScalarParameter() {defaultValue=9, parameterName=Bloom Zone Count}, B=@n7)
  n9 = MaterialFunctionCall(Input=@n4, Input Low=@n5, Input High=@n6, Target Low=@n7, Target High=@n8) {materialFunction=&RemapValueRange}
  Frac(None=@n9.Result)
  n10 = MaterialFunctionCall(Float2=@n1) {materialFunction=&BreakOutFloat2Components}
  n11 = Constant()
  n12 = ScalarParameter() {defaultValue=0.1, parameterName=V Edge Width}
  n13 = Constant() {r=1}
  n14 = Max(
   A=OneMinus(None=MaterialFunctionCall(Input=@n10.R, Input Low=@n11, Input High=@n12) {materialFunction=&RemapValue_0-1}.Result)
   B=MaterialFunctionCall(Input=@n10.R, Input Low=OneMinus(None=@n12), Input High=@n13) {materialFunction=&RemapValue_0-1}.Result
  )
  n15 = Saturate(None=@n14)
  n16 = Constant()
  n17 = Constant() {r=0.4}
  n18 = ComponentMask(None=NamedRerouteUsage()) {g=true}
  n19 = MaterialFunctionCall(Input=Abs(None=ConstantBiasScale(None=@n18) {bias=-0.5, scale=2}), Input Low=Constant(), Input High=Constant() {r=0.1}) {materialFunction=&RemapValue_0-1}
  n20 = Multiply(
   A=OneMinus(None=MaterialFunctionCall(Input=@n15, Input Low=@n16, Input High=@n17) {materialFunction=&RemapValue_0-1}.Result)
   B=Power(Base=OneMinus(None=@n19.Result)) {constExponent=0.3}
  )
  n21 = ComponentMask(None=@n3) {g=true}
  n22 = Abs(None=ConstantBiasScale(None=@n21) {bias=-0.5, scale=2})
  n23 = Saturate(None=InverseLinearInterpolate(Value=@n22) {constB=0.3})
  n24 = OneMinus(None=@n23)
  n25 = Power(Base=Saturate(None=InverseLinearInterpolate(Value=@n24) {constB=0.8}))
  n26 = Multiply(A=@n4, B=Subtract(A=@n8, B=@n7))
  n27 = Multiply(A=Time(), B=ScalarParameter() {defaultValue=3, parameterName=Blink Speed})
  n28 = Floor(None=ScalarParameter() {defaultValue=4, parameterName=Light Blink Spacing})
  n29 = Divide(B=@n28)
  n30 = Multiply(A=Add(A=Subtract(A=@n26, B=@n7), B=Floor(None=@n27)), B=@n29)
  n31 = Frac(None=@n30)
  n32 = Clamp(None=Floor(None=ScalarParameter() {defaultValue=1, parameterName=Light Blink Width}), Max=@n28)
  n33 = Saturate(None=Divide(A=Multiply(A=@n31, B=@n28), B=@n32))
  n34 = StaticSwitchParameter(True=@n25, False=Multiply(A=@n25, B=Floor(None=@n33))) {parameterName=Scrolling Lights}
  n35 = Frac(None=Add(A=@n30, B=Divide(A=@n29)) {constB=0})
  n36 = ConstantBiasScale(None=Divide(A=Multiply(A=@n35, B=@n28), B=Add(A=@n32))) {bias=-0.5, scale=2}
  n37 = Multiply(A=OneMinus(None=Abs(None=@n36)), B=@n32)
  n38 = Power(Base=OneMinus(None=Saturate(None=@n37))) {constExponent=1}
  LinearInterpolate(
   A=Constant3Vector() {constant=<rgba 1 0.4839 0.9524 1>}
   B=LinearInterpolate(
    A=Constant3Vector() {constant=<rgba 1 0.561 0.9176 1>}
    B=Constant3Vector() {constant=<rgba 1 0.9874 0.9177 1>}
    Alpha=Multiply(A=@n20, B=@n34)
   )
   Alpha=@n38
  )
  n39 = DotProduct(A=VertexNormalWS(), B=Multiply(A=CameraVectorWS()) {constB=-1})
  n40 = Power(Base=Saturate(None=Abs(None=@n39))) {constExponent=0.7}
  n41 = ScalarParameter() {defaultValue=0.5, parameterName=U Edge Width}
  n42 = OneMinus(None=MaterialFunctionCall(Input=@n10.G, Input Low=@n11, Input High=@n41) {materialFunction=&RemapValue_0-1}.Result)
  n43 = MaterialFunctionCall(Input=@n10.G, Input Low=OneMinus(None=@n41), Input High=@n13) {materialFunction=&RemapValue_0-1}
  n44 = Length(None=AppendVector(A=Max(A=@n42, B=@n43.Result), B=@n14)) {is=class MaterialExpressionLength}
  n45 = Power(Base=OneMinus(None=Saturate(None=@n44))) {constExponent=1.5}
  n46 = Multiply(A=ScalarParameter() {defaultValue=0.1333, parameterName=Bracket Edge Padding}) {constB=-1}
  n47 = MaterialFunctionCall(
   Input=@n4
   Input Low=@n5
   Input High=@n6
   Target Low=@n46
   Target High=Subtract(A=ScalarParameter() {defaultValue=5, parameterName=Bracket Count}, B=@n46)
  ) {materialFunction=&RemapValueRange}
  n48 = Abs(None=ConstantBiasScale(None=Frac(None=@n47.Result)) {bias=-0.5, scale=2})
  n49 = LinearInterpolate(Alpha=@n22) {constA=6, constB=3}
  n50 = Power(Base=Saturate(None=Multiply(A=@n48, B=@n49) {constB=5})) {constExponent=0.7}
  n51 = Constant() {r=1}
  n52 = Multiply(A=Time(), B=ScalarParameter() {defaultValue=8, parameterName=Light Scroll Speed})
  n53 = Floor(None=@n52)
  n54 = ScalarParameter() {defaultValue=9, parameterName=Lights Per Bloom Zone}
  n55 = Subtract(A=Add(A=@n9.Result, B=Divide(A=@n53, B=@n54)), B=ScalarParameter() {defaultValue=0.4, parameterName=Bloom Zone Lit Ratio})
  n56 = LinearInterpolate(A=ScalarParameter() {defaultValue=0.5, parameterName=UV Gap X, sliderMax=1}, Alpha=@n22) {constB=0}
  n57 = Frac(None=Subtract(A=@n55, B=@n56))
  n58 = Divide(B=OneMinus(None=@n56))
  n59 = Frac(None=Min(A=Multiply(A=@n57, B=@n58)))
  n60 = Min(A=Multiply(A=Frac(None=@n55), B=@n58)) {constB=0.5}
  n61 = ScalarParameter() {defaultValue=0.1, parameterName=UV Gap Y}
  n62 = Frac(None=Subtract(A=@n21, B=@n61))
  n63 = Divide(B=OneMinus(None=@n61))
  n64 = Frac(None=Min(A=Multiply(A=@n62, B=@n63)))
  n65 = Min(A=Multiply(A=Frac(None=@n21), B=@n63)) {constB=0.5}
  n66 = AppendVector(A=Add(A=Floor(None=@n55), B=Max(A=@n59, B=@n60)), B=Add(A=Floor(None=@n21), B=Max(A=@n64, B=@n65)))
  n67 = Constant() {r=1}
  n68 = Constant()
  n69 = Constant() {r=1}
  n70 = Constant() {r=1.5}
  n71 = Constant() {r=0.3}
  n72 = Constant() {r=1}
  n73 = Constant() {r=0.1}
  n74 = MaterialFunctionCall!(
   Light Spread=@n51
   Light Grid UVs=@n66
   Segment Width Count=@n67
   Unlit Width Count=@n68
   Sample Offset
   Light Count Y=@n69
   Light Spread Power=@n70
   Light Core Width=@n71
   Light Core Power=@n72
   Core Depth=@n73
  ) {materialFunction=/Game/Environments/Helios/Props/RustBucket/RustBucket_Pipeline_A/MeshLights/Materials/MaterialFunctions/MF_CalculatePipelineLED}
  n75 = MaterialFunctionCall(
   Light Spread=@n51
   Light Grid UVs=@n66
   Segment Width Count=@n67
   Unlit Width Count=@n68
   Sample Offset=Constant2Vector() {r=-1}
   Light Count Y=@n69
   Light Spread Power=@n70
   Light Core Width=@n71
   Light Core Power=@n72
   Core Depth=@n73
  ) {materialFunction=&MF_CalculatePipelineLED}
  n76 = MaterialFunctionCall(
   Light Spread=@n51
   Light Grid UVs=@n66
   Segment Width Count=@n67
   Unlit Width Count=@n68
   Sample Offset=Constant2Vector() {r=1}
   Light Count Y=@n69
   Light Spread Power=@n70
   Light Core Width=@n71
   Light Core Power=@n72
   Core Depth=@n73
  ) {materialFunction=&MF_CalculatePipelineLED}
  n77 = Add(A=Add(A=@n74.Spread, B=@n75.Spread), B=@n76.Spread)
  n78 = Multiply(A=Multiply(A=Multiply(A=@n40, B=@n45), B=LinearInterpolate(Alpha=@n50) {constA=0.2}), B=@n77)
  n79 = CollectionParameter() {
   collection=&/MPC_PipelineLight
   parameterName=Brightness Multiplier
  }
  n80 = Subtract(A=Time(), B=ScalarParameter() {bUseCustomPrimitiveData=true, parameterName=TimeFromBP, primitiveDataIndex=1})
  n81 = ScalarParameter() {defaultValue=0.5, parameterName=Flare Speed}
  n82 = CurveAtlasRowParameter(CurveTime=Saturate(None=Divide(A=@n80, B=@n81))) {
   atlas=/Game/Environments/Helios/Props/RustBucket/RustBucket_Pipeline_A/MeshLights/Curves/CA_PipelineLights
   curve=/Game/Environments/Helios/Props/RustBucket/RustBucket_Pipeline_A/MeshLights/Curves/C_LightPulseOn
  }
  n83 = Multiply(A=@n79, B=@n82.R)
  Multiply(A=Multiply(A=@n78) {constB=0.5}, B=Saturate(None=@n83))
  n84 = InverseLinearInterpolate(Value=@n24) {constA=0.6}
  Multiply(A=Power(Base=Saturate(None=@n84)) {constExponent=0.5}, B=OneMinus(None=@n38))
  Multiply(A=Multiply(A=@n77, B=Constant3Vector() {constant=<rgba 1 0.2188 0.6164 1>}), B=@n83)
  n85 = TextureCoordinate()
  n86 = ComponentMask(None=@n85) {r=true}
  n87 = OneMinus(None=@n86)
  n88 = Floor(None=ScalarParameter() {bUseCustomPrimitiveData=true, parameterName=Direction, sliderMax=1})
  NamedRerouteDeclaration(None=AppendVector(A=LinearInterpolate(A=@n86, B=@n87, Alpha=@n88), B=ComponentMask(None=@n85) {g=true})) {name=Directional TexCoord, nodeColor=<rgba 0 0.3437 1 1>}
  OneMinus(None=@n22)
  Power(Base=@n78)
  n89 = StaticSwitchParameter(True=Add(A=@n9.Result, B=@n52), False=@n9.Result) {parameterName=Scrolling Lights}
  n90 = Abs(None=ConstantBiasScale(None=Frac(None=@n89)) {bias=-0.5, scale=2})
  Saturate(None=Length(None=AppendVector(A=@n23, B=@n90)))
  StaticSwitchParameter(True=@n50, False=Multiply(A=@n50, B=@n38)) {parameterName=Scrolling Lights}
 ]
}
