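# Purple LED light strip material for the Helios RustBucket pipeline, drawing a dot grid that steps along in lit and unlit segments.
# Nine MF_CalculatePipelineLED calls sum each cell and its neighbours into normalised core and spread glows, the core weighted by view angle.
# A collection brightness, a curve-atlas pulse timed from TimeFromBP custom primitive data, and a glass tint scale the glow.
# Custom primitive data also supplies a UV direction flip.
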Material /Game/Environments/Helios/Props/RustBucket/RustBucket_Pipeline_A/MeshLights/Materials/M_PipelineLightStrip_Purple
graph {
 properties {bAutomaticallySetUsageInEditor=false, pixelDepthOffsetMode=PDOM_Legacy, usedShadingModels=Default Lit}
 nodes [
  n1 = ScalarParameter() {defaultValue=0.5, parameterName=Light Core Width}
  n2 = ScalarParameter() {defaultValue=5, parameterName=Light Core Power}
  n3 = Power(Base=Saturate(None=Subtract(A=@n1)), Exp=@n2)
  n4 = Power(Base=Saturate(None=Subtract(A=@n1) {constB=1.4142}), Exp=@n2)
  n5 = Add(B=Add(A=Multiply(A=@n3) {constB=4}, B=Multiply(A=@n4) {constB=4}) {constA=1}) {constA=1}
  n6 = ScalarParameter() {defaultValue=1.5, parameterName=Light Spread Width, sliderMax=1.5}
  n7 = ScalarParameter() {defaultValue=3, parameterName=Light Spread Power}
  n8 = Power(Base=Saturate(None=Subtract(A=@n6)), Exp=@n7)
  n9 = Power(Base=Saturate(None=Subtract(A=@n6) {constB=1.4142}), Exp=@n7)
  n10 = Add(B=Add(A=Multiply(A=@n8) {constB=4}, B=Multiply(A=@n9) {constB=4}) {constA=1}) {constA=1}
  LinearInterpolate(A=@n5, B=@n10)
  n11 = ScalarParameter() {defaultValue=1.5, parameterName=Light Spread Width, sliderMax=1.5}
  n12 = NamedRerouteUsage()
  n13 = NamedRerouteUsage()
  n14 = NamedRerouteUsage()
  n15 = NamedRerouteUsage()
  n16 = ScalarParameter() {defaultValue=3, parameterName=Light Spread Power}
  n17 = ScalarParameter() {defaultValue=0.5, parameterName=Light Core Width}
  n18 = ScalarParameter() {defaultValue=5, parameterName=Light Core Power}
  n19 = ScalarParameter() {defaultValue=0.3, parameterName=Core Depth}
  n20 = MaterialFunctionCall!(
   Light Spread=@n11
   Light Grid UVs=@n12
   Segment Width Count=@n13
   Unlit Width Count=@n14
   Sample Offset
   Light Count Y=@n15
   Light Spread Power=@n16
   Light Core Width=@n17
   Light Core Power=@n18
   Core Depth=@n19
  ) {materialFunction=/Game/Environments/Helios/Props/RustBucket/RustBucket_Pipeline_A/MeshLights/Materials/MaterialFunctions/MF_CalculatePipelineLED}
  n21 = MaterialFunctionCall(
   Light Spread=@n11
   Light Grid UVs=@n12
   Segment Width Count=@n13
   Unlit Width Count=@n14
   Sample Offset=Constant2Vector() {g=1}
   Light Count Y=@n15
   Light Spread Power=@n16
   Light Core Width=@n17
   Light Core Power=@n18
   Core Depth=@n19
  ) {materialFunction=&MF_CalculatePipelineLED}
  n22 = MaterialFunctionCall(
   Light Spread=@n11
   Light Grid UVs=@n12
   Segment Width Count=@n13
   Unlit Width Count=@n14
   Sample Offset=Constant2Vector() {r=1}
   Light Count Y=@n15
   Light Spread Power=@n16
   Light Core Width=@n17
   Light Core Power=@n18
   Core Depth=@n19
  ) {materialFunction=&MF_CalculatePipelineLED}
  n23 = MaterialFunctionCall(
   Light Spread=@n11
   Light Grid UVs=@n12
   Segment Width Count=@n13
   Unlit Width Count=@n14
   Sample Offset=Constant2Vector() {r=-1}
   Light Count Y=@n15
   Light Spread Power=@n16
   Light Core Width=@n17
   Light Core Power=@n18
   Core Depth=@n19
  ) {materialFunction=&MF_CalculatePipelineLED}
  n24 = MaterialFunctionCall(
   Light Spread=@n11
   Light Grid UVs=@n12
   Segment Width Count=@n13
   Unlit Width Count=@n14
   Sample Offset=Constant2Vector() {g=-1}
   Light Count Y=@n15
   Light Spread Power=@n16
   Light Core Width=@n17
   Light Core Power=@n18
   Core Depth=@n19
  ) {materialFunction=&MF_CalculatePipelineLED}
  n25 = MaterialFunctionCall(
   Light Spread=@n11
   Light Grid UVs=@n12
   Segment Width Count=@n13
   Unlit Width Count=@n14
   Sample Offset=Constant2Vector() {g=1, r=1}
   Light Count Y=@n15
   Light Spread Power=@n16
   Light Core Width=@n17
   Light Core Power=@n18
   Core Depth=@n19
  ) {materialFunction=&MF_CalculatePipelineLED}
  n26 = MaterialFunctionCall(
   Light Spread=@n11
   Light Grid UVs=@n12
   Segment Width Count=@n13
   Unlit Width Count=@n14
   Sample Offset=Constant2Vector() {g=-1, r=1}
   Light Count Y=@n15
   Light Spread Power=@n16
   Light Core Width=@n17
   Light Core Power=@n18
   Core Depth=@n19
  ) {materialFunction=&MF_CalculatePipelineLED}
  n27 = MaterialFunctionCall(
   Light Spread=@n11
   Light Grid UVs=@n12
   Segment Width Count=@n13
   Unlit Width Count=@n14
   Sample Offset=Constant2Vector() {g=1, r=-1}
   Light Count Y=@n15
   Light Spread Power=@n16
   Light Core Width=@n17
   Light Core Power=@n18
   Core Depth=@n19
  ) {materialFunction=&MF_CalculatePipelineLED}
  n28 = Add(
   A=Add(A=Add(A=@n20.Core, B=@n21.Core), B=Add(A=@n22.Core, B=@n23.Core))
   B=Add(A=Add(A=@n24.Core, B=@n25.Core), B=Add(A=@n26.Core, B=@n27.Core))
  )
  n29 = MaterialFunctionCall(
   Light Spread=@n11
   Light Grid UVs=@n12
   Segment Width Count=@n13
   Unlit Width Count=@n14
   Sample Offset=Constant2Vector() {g=-1, r=-1}
   Light Count Y=@n15
   Light Spread Power=@n16
   Light Core Width=@n17
   Light Core Power=@n18
   Core Depth=@n19
  ) {materialFunction=&MF_CalculatePipelineLED}
  n30 = Multiply(A=Divide(A=Add(A=@n28, B=@n29.Core), B=@n5), B=ScalarParameter() {defaultValue=10, parameterName=Core Brightness})
  n31 = VectorParameter() {parameterName=Core Color, defaultValue=<rgba 0.2027 0.7349 1 1>}
  n32 = Transform(None=CameraVectorWS()) {transformSourceType=TRANSFORMSOURCE_World, transformType=TRANSFORM_Tangent}
  n33 = Constant3Vector() {constant=<rgba 0 0 1 0>}
  n34 = Add(A=Add(A=@n20.Spread, B=@n21.Spread), B=Add(A=@n22.Spread, B=@n23.Spread))
  n35 = Add(A=Add(A=@n24.Spread, B=@n25.Spread), B=Add(A=@n26.Spread, B=@n27.Spread))
  n36 = Divide(A=Add(A=Add(A=@n34, B=@n35), B=@n29.Spread), B=@n10)
  n37 = ScalarParameter() {defaultValue=7, parameterName=Spread Brightness}
  n38 = Add(
   A=Multiply(A=Multiply(A=@n30, B=@n31.RGB), B=DotProduct(A=@n32, B=@n33))
   B=Multiply(A=Multiply(A=@n36, B=@n37), B=VectorParameter() {parameterName=Spread Color, defaultValue=<rgba 0.9536 0.7121 1 1>}.RGB)
  )
  n39 = CollectionParameter() {
   collection=&/MPC_PipelineLight
   parameterName=Brightness Multiplier
  }
  n40 = Subtract(A=Time(), B=ScalarParameter() {bUseCustomPrimitiveData=true, parameterName=TimeFromBP, primitiveDataIndex=1})
  n41 = ScalarParameter() {defaultValue=0.5, parameterName=Pulse Speed}
  n42 = CurveAtlasRowParameter(CurveTime=Saturate(None=Divide(A=@n40, B=@n41))) {
   atlas=/Game/Environments/Helios/Props/RustBucket/RustBucket_Pipeline_A/MeshLights/Curves/CA_PipelineLights
   curve=/Game/Environments/Helios/Props/RustBucket/RustBucket_Pipeline_A/MeshLights/Curves/C_LightPulseOn
  }
  Multiply(
   A=Multiply(A=Multiply(A=@n38, B=@n39), B=@n42.R)
   B=VectorParameter() {parameterName=Glass Color, defaultValue=<rgba 0.0851 0.0345 0.0724 1>}.RGB
  )
  n43 = Floor(None=Max(A=ScalarParameter() {defaultValue=6, parameterName=Lit Width Count}) {constB=0})
  n44 = Floor(None=Max(A=ScalarParameter() {defaultValue=3, parameterName=Unlit Width Count}) {constB=0})
  NamedRerouteDeclaration(None=Add(
    A=NamedRerouteDeclaration(None=@n43) {name=Fixed Lit Width Count, nodeColor=<rgba 0.9219 1 0 1>}
    B=NamedRerouteDeclaration(None=@n44) {name=Fixed Unlit Width Count, nodeColor=<rgba 0.7578 0 1 1>}
   )) {name=Segment Width Count, nodeColor=<rgba 1 0 0.2812 1>}
  n45 = MaterialFunctionCall(Float2=NamedRerouteUsage()) {materialFunction=&BreakOutFloat2Components}
  n46 = MaterialFunctionCall() {materialFunction=&ObjectScale}
  n47 = ScalarParameter() {defaultValue=3, parameterName=Light Count Y, sliderMax=7, sliderMin=1}
  n48 = NamedRerouteDeclaration(None=Floor(None=Max(A=@n47))) {name=Fixed Light Count Y, nodeColor=<rgba 1 0 0.8437 1>}
  n49 = Multiply(A=Multiply(A=@n45.R, B=Divide(A=@n46.Scale X, B=@n46.Scale X)), B=@n48)
  n50 = Time()
  n51 = ScalarParameter() {defaultValue=8, parameterName=Blink Speed}
  n52 = Floor(None=Modulo!(A=Multiply(A=@n50, B=@n51), B=NamedRerouteUsage()))
  NamedRerouteDeclaration(None=AppendVector(A=Add(A=@n49, B=@n52), B=Multiply(A=@n45.G, B=@n48))) {name=Light Grid UVs, nodeColor=<rgba 0 0.6016 1 1>}
  n53 = TextureCoordinate()
  n54 = ComponentMask(None=@n53) {r=true}
  n55 = OneMinus(None=@n54)
  n56 = Floor(None=ScalarParameter() {bUseCustomPrimitiveData=true, parameterName=Direction, sliderMax=1})
  NamedRerouteDeclaration(None=AppendVector(A=LinearInterpolate(A=@n54, B=@n55, Alpha=@n56), B=ComponentMask(None=@n53) {g=true})) {name=Directional TexCoord, nodeColor=<rgba 0.2656 1 0 1>}
  n57 = MaterialFunctionCall(Float2=Floor(None=NamedRerouteUsage())) {materialFunction=&BreakOutFloat2Components}
  n58 = NamedRerouteUsage()
  n59 = OneMinus(None=Modulo(A=@n58, B=Constant() {r=2}))
  n60 = Subtract(A=@n58, B=@n59)
  n61 = Subtract(A=@n57.G, B=Floor(None=Divide(A=@n60)))
  n62 = Multiply(A=Subtract(A=Saturate(None=@n61)), B=@n59)
  n63 = Add(A=@n57.R, B=Abs(None=Add(A=@n61, B=@n62)))
  n64 = NamedRerouteUsage()
  n65 = NamedRerouteUsage()
  Saturate(None=Subtract(A=Modulo(A=@n63, B=@n64), B=Subtract(A=@n65)))
  ScalarParameter() {defaultValue=0.7, parameterName=Core Spread Mix, sliderMax=1}
 ]
}
