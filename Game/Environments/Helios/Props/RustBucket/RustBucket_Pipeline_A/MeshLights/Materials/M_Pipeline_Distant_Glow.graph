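# Unlit translucent material for a distant glow on Helios RustBucket pipeline lights.
# A UV box-distance falloff shapes the soft glow, with an aspect term that divides object Scale X by itself.
# A camera-distance sphere mask hides it up close and varies edge sharpness, while a view-angle term fades grazing faces.
# Glow color is scaled by collection brightness and a curve-atlas pulse timed from Blueprint-fed custom primitive data.
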Material /Game/Environments/Helios/Props/RustBucket/RustBucket_Pipeline_A/MeshLights/Materials/M_Pipeline_Distant_Glow
graph {
 properties {
  bAutomaticallySetUsageInEditor=false
  blendMode=BLEND_Translucent
  pixelDepthOffsetMode=PDOM_Legacy
  shadingModel=MSM_Unlit
  usedShadingModels=Unlit
 }
 nodes [
  n1 = TextureCoordinate()
  n2 = MaterialFunctionCall(Float2=Abs(None=ConstantBiasScale(None=@n1) {bias=-0.5, scale=2})) {materialFunction=&BreakOutFloat2Components}
  n3 = MaterialFunctionCall() {materialFunction=&ObjectScale}
  n4 = Divide(A=@n3.Scale X, B=@n3.Scale X)
  n5 = AppendVector(A=Multiply(A=@n2.R, B=@n4), B=@n2.G)
  n6 = ScalarParameter() {defaultValue=1, parameterName=MaxFadeWidth}
  n7 = AppendVector(A=Subtract(A=@n4, B=@n6), B=OneMinus(None=@n6))
  n8 = Length(None=Max(A=Subtract(A=@n5, B=@n7)) {constB=0}) {is=class MaterialExpressionLength}
  n9 = Divide(B=@n6)
  n10 = Saturate(None=NamedRerouteDeclaration(None=Multiply(A=@n8, B=@n9)) {name=DistanceFromBox, nodeColor=<rgba 1 0.2812 0 1>})
  n11 = CameraPositionWS()
  n12 = WorldPosition()
  n13 = ScalarParameter() {defaultValue=30000, parameterName=Fade Radius}
  n14 = Power(Base=OneMinus(None=SphereMask(A=@n11, B=@n12.XYZ, Radius=@n13) {attenuationRadius=10000, hardnessPercent=35}))
  n15 = MaterialFunctionCall(Input=DotProduct(A=VertexNormalWS(), B=CameraVectorWS()), Input Low=Constant() {r=0.2}, Input High=Constant() {r=0.5}) {materialFunction=&RemapValue_0-1}
  Multiply(
   A=Power(Base=OneMinus(None=@n10), Exp=LinearInterpolate(Alpha=@n14) {constA=6, constB=3}) {constExponent=3}
   B=Multiply(A=@n14, B=Saturate(None=@n15.Result))
  )
  n16 = CollectionParameter() {
   collection=&/MPC_PipelineLight
   parameterName=Brightness Multiplier
  }
  n17 = Subtract(A=Time(), B=ScalarParameter() {bUseCustomPrimitiveData=true, parameterName=TimeFromBP, primitiveDataIndex=1})
  n18 = ScalarParameter() {defaultValue=0.5, parameterName=Pulse Speed}
  n19 = CurveAtlasRowParameter(CurveTime=Saturate(None=Divide(A=@n17, B=@n18))) {
   atlas=/Game/Environments/Helios/Props/RustBucket/RustBucket_Pipeline_A/MeshLights/Curves/CA_PipelineLights
   curve=/Game/Environments/Helios/Props/RustBucket/RustBucket_Pipeline_A/MeshLights/Curves/C_LightPulseOn
   parameterName=PulseCurve
  }
  Multiply(A=Multiply(A=VectorParameter() {parameterName=GlowColor, defaultValue=<rgba 1 0.5117 0 1>}.RGB, B=@n16), B=@n19.R)
 ]
}
